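# Unlit alpha-composite material that draws a team-color player outline and optional halo glow on a camera-facing mesh biased along the camera direction.
# Four diagonal SceneDepth taps form a Sobel-style edge, masked by a radial falloff and depth behind the translucent surface.
# Medium and High quality add DepthFade to the edge, and the UseGlow switch adds the depth-masked halo.
# Custom primitive data supplies the color, letting each primitive tint independently.
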Material /Game/Creative/Abilities/TeamColor/M_PlayerOutline_02
graph {
 properties {bEnableResponsiveAA=true, blendMode=BLEND_AlphaComposite, pixelDepthOffsetMode=PDOM_Legacy, shadingModel=MSM_Unlit, usedShadingModels=Unlit}
 nodes [
  n1 = Subtract(A=CameraPositionWS(), B=ObjectPositionWS())
  Add(
   A=MaterialFunctionCall!(
    WorldPosition=WorldPosition().XYZ
    PivotLocation
    WS Normals
    Custom Object Basis 1=Constant3Vector() {constant=<rgba 0 1 0 1>}
    Custom Object Basis 2=Constant3Vector() {constant=<rgba 0 0 -1 0>}
    Custom Object Basis 3
   ) {materialFunction=/Engine/Functions/Engine_MaterialFunctions02/WorldPositionOffset/AlignMeshToTheCamera}.World Position Offset
   B=Multiply(A=Normalize(VectorInput=@n1), B=ScalarParameter() {defaultValue=25, parameterName=CameraOffsetBias})
  )
  Clamp(Max=ScalarParameter() {defaultValue=200, parameterName=SceneDepthClampMaxMobile}) {clampMode=CMODE_ClampMax}
  Divide(A=SceneTexture() {sceneTextureId=PPI_SceneDepth}.Color, B=ScalarParameter() {defaultValue=2, parameterName=EdgeSceneDepthDivisorMobile})
  MaterialFunctionCall!(In=Saturate(None=Divide())) {materialFunction=/Game/Effects/Fort_Effects/Materials/_Functions/MF_Smoothstep_01}
  n2 = VectorParameter() {bUseCustomPrimitiveData=true, parameterName=OutlineColorAndGlowScale, defaultValue=<rgba 1 0.096 0.68 0>}
  n3 = TextureCoordinate()
  n4 = Constant() {r=0.5}
  n5 = ScalarParameter() {parameterName=HaloSphereMaskHardness}
  n6 = MaterialFunctionCall!(
   Distance Bias=ScalarParameter() {defaultValue=25, parameterName=CameraOffsetBias}
   Distance Divisor=ScalarParameter() {defaultValue=45, parameterName=DistanceDivisor}
   Scene Depth
   Pixel Depth
  ) {materialFunction=/Game/Effects/Fort_Effects/Materials/_Functions/MF_DistanceBehindTranslucency_01}
  n7 = Multiply(
   A=Multiply(A=SphereMask(A=@n3, B=@n4, Radius=@n4, Hardness=@n5), B=ScalarParameter() {defaultValue=0.025, parameterName=GlowScale})
   B=@n6.Clamped
  )
  n8 = ScreenPosition()
  n9 = SceneTexelSize()
  n10 = Add(A=@n8.ViewportUV, B=Multiply(A=@n9, B=Constant2Vector() {g=-1, r=-1}))
  n11 = SceneTexture() {sceneTextureId=PPI_SceneDepth}
  n12 = Add(A=@n8.ViewportUV, B=Multiply(A=@n9, B=Constant2Vector() {g=-1, r=1}))
  n13 = Add(
   A=Subtract(A=SceneTexture(UVs=@n10) {sceneTextureId=PPI_SceneDepth}.Color, B=@n11.Color)
   B=Subtract(A=SceneTexture(UVs=@n12) {sceneTextureId=PPI_SceneDepth}.Color, B=@n11.Color)
  )
  n14 = Multiply(A=@n9, B=Constant2Vector() {g=1, r=1})
  n15 = Subtract(A=SceneTexture(UVs=Add(A=@n8.ViewportUV, B=@n14)) {sceneTextureId=PPI_SceneDepth}.Color, B=@n11.Color)
  n16 = Constant2Vector() {g=1, r=-1}
  n17 = SceneTexture(UVs=Add(A=@n8.ViewportUV, B=Multiply(A=@n9, B=@n16))) {sceneTextureId=PPI_SceneDepth}
  n18 = Max(B=Add(A=Add(A=@n13, B=@n15), B=Subtract(A=@n17.Color, B=@n11.Color)))
  n19 = ScalarParameter() {defaultValue=20, parameterName=EdgeSobelDivisor}
  n20 = Round(None=Saturate(None=Divide(A=@n18, B=@n19)))
  n21 = Constant() {r=0.5}
  n22 = Multiply(
   A=Multiply(
    A=ComponentMask(None=@n20) {r=true}
    B=MaterialFunctionCall(In=@n6.Inverted) {materialFunction=&MF_Smoothstep_01}.Result
   )
   B=SphereMask(A=TextureCoordinate(), B=@n21, Radius=@n21, Hardness=Constant() {r=0.9})
  )
  n23 = DepthFade(Opacity=@n22, FadeDistance=ScalarParameter() {defaultValue=5, parameterName=DepthFadeDistance})
  n24 = QualitySwitch(Default=@n22, Low=@n22, High=@n23, Medium=@n23)
  Multiply(A=@n2.RGB, B=StaticSwitchParameter(True=Add(A=@n7, B=@n24), False=@n24) {defaultValue=true, parameterName=UseGlow})
 ]
}
